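# NPC conversation for the Season 33 Battle Pass shotgun weapon expertise completion quest, praising the player and announcing a high-tier shotgun sale.
# The entry tag leads into a linear chain of three speech nodes, ending with a warning that continued proof is needed to keep access.
# The opening "Shotgun Expertise" choice, tagged as main questline, requires that quest be active.
# The final Continue choice carries the quest's conversation tag.
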ConversationDatabase /BattlePassS33_Quests/Conversations/WeaponExpertise/Conversation_S33_Weaponexpertise_Shotgun_Complete_01
entry_points [{tag=Conversation.Entry.Season33.Weaponexpertise.Shotgun.Complete.01, destinations [#1]}]
nodes [
 ConversationEntryPointNode_0 {
  id=#1
  class=ConversationEntryPointNode
  outputs [#2]
  properties {entryTag {tagName=Conversation.Entry.Season33.Weaponexpertise.Shotgun.Complete.01}}
 }
 Speech_C_3 {
  id=#2
  class=Speech_C
  outputs [#3]
  properties {
   softTaskIcon=&T_UI_NPC_WeaponExpertise
   generalConfig {defaultMessage="By the dragon's breath! You're astounding."}
  }
  sub_nodes [
   ConversationChoiceNode_5 {
    class=ConversationChoiceNode
    properties {defaultChoiceDisplayText=Shotgun Expertise, choiceTags {gameplayTags [{tagName=Conversation.Choice.Category.MainQuestline}]}}
   }
   FortConversationRequirement_ControllerRequirement_4 {
    class=FortConversationRequirement_ControllerRequirement
    properties {
     nodeName=Require: Is Quest Active
     requirement {
      class=FortControllerRequirement_IsQuestActive
      properties {questDefinition=/BattlePassS33_Quests/Items/QuestItems/WeaponExpertise/Shotgun/quest_s33_weaponexpertise_shotgun_complete_01}
     }
    }
   }
  ]
 }
 Speech_C_28 {
  id=#3
  class=Speech_C
  outputs [#4]
  properties {generalConfig {defaultMessage="For a small fee, I can now sell you a high-tier shotgun."}}
  sub_nodes [ConversationChoiceNode_6 {class=ConversationChoiceNode, properties {defaultChoiceDisplayText=Continue}}]
 }
 Speech_C_1 {
  id=#4
  class=Speech_C
  outputs []
  properties {generalConfig {defaultMessage="But fair warning, you will need to continue proving yourself to retain access to this sale."}}
  sub_nodes [ConversationChoiceNode_8 {
    class=ConversationChoiceNode
    properties {defaultChoiceDisplayText=Continue, choiceTags {gameplayTags [{tagName=Athena.Quests.S33.Conversation.Weaponexpertise.Shotgun.Complete.01}]}}
   }]
 }
]
